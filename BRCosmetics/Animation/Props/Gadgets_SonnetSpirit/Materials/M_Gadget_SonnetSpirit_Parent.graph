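# Parent material for the Sonnet Spirit gadget, a cloth-shaded anime character-style material with a progressively revealed detail driven by SS_SmileProgress, apparently a drawn smile.
# It wraps the 2023 anime shader's Pre, Calculations and Post functions.
# The progress parameter offsets a texture mask's green channel into a smoothstep reveal, which tints lit and shaded color dark blue within the red channel and flattens unrevealed normals within the blue channel.
# The Time-based DrawingSpeed branch is unconnected.
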Material /BRCosmetics/Animation/Props/Gadgets_SonnetSpirit/Materials/M_Gadget_SonnetSpirit_Parent
graph {
 properties {
  bAutomaticallySetUsageInEditor=false
  bUseMaterialAttributes=true
  bUsedWithMeshDeformer=true
  bUsedWithSkeletalMesh=true
  pixelDepthOffsetMode=PDOM_Legacy
  shadingModel=MSM_Cloth
  usedShadingModels=Cloth
 }
 nodes [
  Constant()
  n1 = MaterialFunctionCall!(CustomUV) {materialFunction=/Game/Global/Materials/Character_Parent_2023/Anime_Parent_2023/MaterialFunctions/MFC_AnimeShader_2023_Pre}
  n2 = Constant3Vector() {constant=<rgba 0 0 1 0>}
  n3 = ScalarParameter() {parameterName=SS_SmileProgress}
  n4 = ConstantBiasScale(None=Frac(None=LinearInterpolate(Alpha=@n3) {constB=0.9999})) {bias=-0.5, scale=2}
  n5 = TextureSample() {samplerType=SAMPLERTYPE_Masks, texture=/BRCosmetics/Animation/Props/Gadgets_SonnetSpirit/Textures/T_SonnetSpirit_Gadgets_TA}
  n6 = SmoothStep(Value=Saturate(None=Add(A=@n4, B=@n5.G))) {constMax=0.6, constMin=0.5}
  n7 = MaterialFunctionCall!(
   CustomLightVector
   OverrideNormal_WS
   STT=@n1.STT
   Normal_Map=LinearInterpolate(A=@n1.Normal, B=LinearInterpolate(A=@n2, B=@n1.Normal, Alpha=@n6), Alpha=@n5.B)
   DFL=@n1.STT
   FX=@n1.STT
  ) {materialFunction=/Game/Global/Materials/Character_Parent_2023/Anime_Parent_2023/MaterialFunctions/MFC_AnimeShader_2023_Calculations}
  n8 = Constant3Vector() {constant=<rgba 0.0015 0.0685 0.2195 1>}
  n9 = Multiply(A=@n6, B=@n5.R)
  MaterialFunctionCall!(
   Shading_Gradient=@n7.Shading_Gradient
   Threshold_Gradient=@n7.Shading_Gradient
   Specular_Gradient=@n7.Shading_Gradient
   Sheen_Mask=@n7.Shading_Gradient
   Thickness=@n7.Shading_Gradient
   DFL_Mask=@n7.Shading_Gradient
   FX=@n7.Shading_Gradient
   Metallic_Gradient=@n7.Shading_Gradient
   RimLight_Gradient=@n7.Shading_Gradient
   NormalBuffer=@n7.Shading_Gradient
   --------
   CINE_LightSources=@n7.Shading_Gradient
   CINE_LightColor=@n7.Shading_Gradient
   ---------
   Color_Lit=LinearInterpolate(A=@n1.Color_Lit, B=@n8, Alpha=@n9)
   Color_Shaded=LinearInterpolate(A=@n1.Color_Shaded, B=@n8, Alpha=@n9)
   SRM=@n1.SRM
   Emissive=@n1.SRM
   DFL_Color=@n1.SRM
   Specular_Color=@n1.SRM
   RimLight_Color=@n1.SRM
  ) {materialFunction=/Game/Global/Materials/Character_Parent_2023/Anime_Parent_2023/MaterialFunctions/MFC_AnimeShader_2023_Post}
  Multiply(A=Time(), B=ScalarParameter() {defaultValue=0.5, parameterName=DrawingSpeed}) {constB=-0.3}
 ]
}
